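# Niagara system for the Lasting Stamina energy effect on a Raptor wildlife mount, drawing motion trails and glows on its skeletal meshes.
# RibbonSourcesLegs pins six particles to bones and emits movement-driven location events while moving or in cinematic mode.
# LegRibbons002 makes BariumPaintColor-tinted ribbons, reading source positions to fade by distance and scaling width by smoothed speed.
# Distance-spawned orange speed lines and short ribbon trails accompany pulsing, mesh-attached SpeedyFeet sprites.
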NiagaraSystem /Energy/FX/NS_Energy_LastingStamina_WildlifeMount_Raptor
dependencies {
 usedDataInterfaces [
  &NiagaraDataInterfaceColorCurve
  &NiagaraDataInterfaceCurve
  &NiagaraDataInterfaceVector2DCurve
  &NiagaraDataInterfaceArrayInt32
  &NiagaraDataInterfaceArrayFloat3
  &NiagaraDataInterfaceArrayFloat
  &NiagaraDataInterfaceParticleRead
 ]
 usedDynamicInputs [&MakeFloatFromVector]
 usedModules [
  &SystemState
  &Timeline
  &self:SystemSpawnScript.SystemScriptSource.SystemScriptGraph.NiagaraNodeAssignment_0.SetVariables_#1
  &self:SystemSpawnScript.SystemScriptSource.SystemScriptGraph.NiagaraNodeAssignment_1.SetVariables_#2
 ]
}
emitters {
 LegRibbons002 {
  emitterUpdateScript [
   &EmitterState()
   /SetVariables(Emitter.RibbonLengthModifier:NiagaraFloat=&Multiply_Float(A=2, B=$User.RibbonLengthFactor)) {isSetParameters=true}
   &InterpolateOverTime(
    Rate Of Change:NiagaraFloat=&MakeCustomFloatFromBool(Bool=&SetBoolByFloatComparison(A=$System.EngineOwnerSpeed, B=0.5), True Float=1, False Float=15)
    Target Float Value:NiagaraFloat=$System.EngineOwnerSpeed
   )
  ]
  particleSpawnScript [
   /SetVariables(Particles.MaxRibbonWidth=15) {isSetParameters=true}
   &InitializeRibbon(
    Color=&MultiplyLinearColor(
     Linear Color=&HueShiftLinearColor(Hue=0, Saturation=0.9, Value=2, Linear Color=$User.BariumPaintColor)
     Scale Factor=<1 1 1 0.05>
    )
    Lifetime=0.2
    Ribbon Width=&FloatFromCurve(
     FloatCurve=DataInterface() {propertyValues {Curve {keys [
         {arriveTangent=4.1785, interpMode=RCIM_Cubic, leaveTangent=4.1785, tangentMode=RCTM_User, time=0, value=0}
         {arriveTangent=0.9995, interpMode=RCIM_Cubic, leaveTangent=0.9995, tangentMode=RCTM_User, time=0.2223, value=0.6825}
         {arriveTangent=-0.0405, interpMode=RCIM_Cubic, leaveTangent=-0.0405, tangentMode=RCTM_User, time=1, value=1}
        ]}}}
     CurveIndex=&ClampFloat(Float=&Divide_Float(A=$Emitter.InterpolateOverTime.InterpolateOverTime_Float.Moving Average, B=100))
     Scale Curve=$Particles.MaxRibbonWidth
    )
    UsePositionOffset=false
    Write Material Random=true
   )
   /SetVariables(Particles.EventPosition:NiagaraPosition=$Particles.Position) {isSetParameters=true}
   /SetVariables(Particles.SpriteSize=<5 5>) {isSetParameters=true}
   /SetVariables(Particles.MaxRibbonLength:NiagaraFloat=&Multiply_Float(A=1, B=$Emitter.RibbonLengthModifier)) {isSetParameters=true}
   /SetVariables() {isSetParameters=true}
   /SetVariables() {isSetParameters=true}
   /SetVariables() {isSetParameters=true}
  ]
  particleUpdateScript [
   &ParticleState()
   &Drag(Drag=50, Use Linear Drag=true, Use Rotational Drag=false)
   &PointAttractionForce(
    Attraction Radius=800
    Attraction Strength=&Multiply_Float(
     A=&Lerp_Float(A=200, B=0, Alpha=$System.NormalizedPlayerSpeed)
     B=&ClampFloat(Float=&Divide_Float(A=$Emitter.InterpolateOverTime.InterpolateOverTime_Float.Moving Average, B=100))
    )
    Attractor Position Offset:Vector3f=<0 0 100>
    Attractor Position Offset Coordinate Space:ENiagaraCoordinateSpace=::Local
    Kill Within Radius=false
    Module.Use Falloff=false
    UseAttractorOffset=true
   ) {enabled=false}
   &SolveForcesAndVelocity(Clamp Velocity=false, Limit Acceleration=false)
   /SetVariables(Particles.NormalizedDistanceFromParent:NiagaraFloat=&ClampFloat(Float=&Divide_Float(
      A=&VectorLength(
       Vector=&Subtract_Vector(
        A=Linked() {linkedVariable {name=Particles.Position, type=NiagaraPosition}}
        B=&self:GetParentParticlePosition(PAR:NiagaraDataInterfaceParticleRead=DataInterface() {propertyValues {EmitterBinding {bindingMode=Other, emitterName=RibbonSourcesLegs}}})
       )
       Fallback Vector=<0 0 0.01>
       Fallback Length=0
      )
      B=$Particles.MaxRibbonLength
     ))) {isSetParameters=true}
   &ScaleColor(
    Scale Alpha=&MakeCustomFloatFromBool(
     Bool=$Particles.EventReceieved
     True Float=&FloatFromCurve(FloatCurve=DataInterface() {propertyValues {Curve {keys [<at 0 1>,<at 0.8 0>]}}}, CurveIndex=$Particles.NormalizedDistanceFromParent)
     False Float=1
    )
    Scale RGB=&VectorFromFloat(Value=&FloatFromCurve(
      FloatCurve=DataInterface() {propertyValues {Curve {keys [
          {arriveTangent=-2.725, interpMode=RCIM_Cubic, leaveTangent=-2.725, tangentMode=RCTM_User, time=0, value=5}
          {arriveTangent=-8.8847, leaveTangent=-8.8847, time=0.2, value=1}
          <at 1 0>
         ]}}}
      CurveIndex=$Particles.NormalizedDistanceFromParent
     ))
    ScaleA=true
    ScaleRGB=true
   )
   &ScaleColor(
    Scale Alpha=&FloatFromCurve(FloatCurve=DataInterface() {propertyValues {Curve {keys [
         {arriveTangent=-3.8552, interpMode=RCIM_Cubic, leaveTangent=-3.8552, tangentMode=RCTM_User, time=0, value=1}
         <at 0.5 0.1>
         {interpMode=RCIM_Cubic, tangentMode=RCTM_User, time=1, value=0.02}
        ]}}})
    Scale RGB=&VectorFromFloat(Value=&FloatFromCurve(
      FloatCurve=DataInterface() {propertyValues {Curve {keys [<at 0 1>,<at 0.1 0.1>]}}}
      CurveIndex=$Particles.NormalizedDistanceFromParent
      Scale Curve=0.4
     ))
    ScaleA=true
    ScaleRGB=true
   ) {module=ScaleColor002}
   /SetVariables(Particles.Color2=&MultiplyLinearColor(
     Linear Color=$User.BariumPaintColor
     Scale Factor=&MakeFloat4FromVectorAndFloat(
      "Vector (RGB)"=&VectorFromFloat(Value=&FloatFromCurve(
        FloatCurve=DataInterface() {propertyValues {Curve {keys [<at 0 1>,<at 0.15 0.1>]}}}
        CurveIndex=$Particles.NormalizedDistanceFromParent
        Scale Curve=0.2
       ))
      "Float (Alpha)"=&Multiply_Float(
       A=&FloatFromCurve(FloatCurve=DataInterface() {propertyValues {Curve {keys [
            {arriveTangent=-3.2069, interpMode=RCIM_Cubic, leaveTangent=-3.2069, tangentMode=RCTM_User, time=0, value=1}
            {interpMode=RCIM_Cubic, tangentMode=RCTM_User, time=1, value=0}
           ]}}})
       B=&FloatFromCurve(
        FloatCurve=DataInterface() {propertyValues {Curve {keys [
            {arriveTangent=-3.6157, interpMode=RCIM_Cubic, leaveTangent=-3.6157, tangentMode=RCTM_User, time=0, value=1}
            <at 0.1 0.1>
            {interpMode=RCIM_Cubic, tangentMode=RCTM_User, time=1, value=0}
           ]}}}
        CurveIndex=$System.NormalizedPlayerSpeed
       )
      )
     )
    )) {isSetParameters=true}
   &KillParticles(Kill Particles=&Boolean_AndOperation(
     Bool1=&SetBoolByFloatComparison(
      A=&MakeFloatFromLinearColor(LinearColor=$Particles.Color, Channel:ENiagaraLinearColor_Channels=::NewEnumerator3)
      B=0.02
      Comparison Type=::NewEnumerator6
     )
     Bool2=$Particles.EventReceieved
    )) {enabled=false, module=KillParticles002}
   /SetVariables(Particles.RibbonU0Override=&MakeCustomFloatFromBool(Bool=$Particles.EventReceieved, True Float=$Particles.NormalizedDistanceFromParent, False Float=0)) {isSetParameters=true}
   &ScaleRibbonWidth(Ribbon Width Scale=&FloatFromCurve(
     FloatCurve=DataInterface() {propertyValues {Curve {keys [{interpMode=RCIM_Cubic, tangentMode=RCTM_User, time=0, value=0}, {interpMode=RCIM_Cubic, tangentMode=RCTM_User, time=1, value=1}]}}}
     CurveIndex=&MakeCustomFloatFromBool(
      Bool=$User.UseCinematicVelocity
      True Float=1
      False Float=&ClampFloat(Float=&Divide_Float(A=$Emitter.InterpolateOverTime.InterpolateOverTime_Float.Moving Average, B=100))
     )
     Scale Curve=&MakeCustomFloatFromBool(
      Bool=$Particles.EventReceieved
      True Float=&FloatFromCurve(
       FloatCurve=DataInterface() {propertyValues {Curve {keys [{tangentMode=RCTM_User, time=0, value=1}, {tangentMode=RCTM_User, time=1, value=1}]}}}
       CurveIndex=$Particles.NormalizedDistanceFromParent
      )
      False Float=1
     )
    ))
  ]
  rendererClasses [&NiagaraRibbonRendererProperties, &NiagaraSpriteRendererProperties]
  renderers [
   {rendererClass=&NiagaraRibbonRendererProperties}
   {rendererClass=&NiagaraSpriteRendererProperties, rendererIndex=1}
   {rendererClass=&NiagaraRibbonRendererProperties, rendererIndex=2}
   {rendererClass=&NiagaraRibbonRendererProperties, rendererIndex=3}
  ]
 }
 RibbonSourcesLegs {
  emitterSpawnScript [
   /SetVariables(Emitter.NumberOfRibbons=6) {isSetParameters=true}
   /SetVariables(Emitter.BoneSocketArray=DataInterface() {propertyValues {GpuSyncMode=SyncCpuToGpu, MaxElements=0, IntData [12,16,18,13,17,19]}}) {isSetParameters=true}
   /SetVariables(Emitter.BoneOffsetArray=DataInterface() {propertyValues {GpuSyncMode=SyncCpuToGpu, MaxElements=0, FloatData [<0 0 0>, <0 0 0>, <0 0 0>, <0 0 0>, <0 0 0>, <0 0 0>]}}) {isSetParameters=true}
   /SetVariables(Emitter.MaxRibbonLengthArray=DataInterface() {propertyValues {GpuSyncMode=SyncCpuToGpu, MaxElements=0, FloatData [200,200,200,200,200,200,200]}}) {isSetParameters=true}
  ]
  emitterUpdateScript [
   &EmitterState(
    Inactive Response=::NewEnumerator1
    Life Cycle Mode=::NewEnumerator1
    Loop Behavior=::NewEnumerator1
    Loop Duration Mode=::NewEnumerator1
    UseLoopDelay=false
   )
   &SpawnBurst_Instantaneous(Spawn Count=$Emitter.NumberOfRibbons, Use Spawn Probability=false, UseLoopCountLimit=false)
  ]
  particleSpawnScript [
   &FN_SetSpawnGroup(Normalize Spawn Group=true, Number of Spawn Groups:NiagaraInt32=$Emitter.NumberOfRibbons, Set Ribbon ID and Ribbon Link Order=true)
   &SkeletalMeshLocation(
    Bone / Socket Index=$Particles.SpawnGroup
    MaskSpawnGroup=false
    Mesh Sampling Type=::NewEnumerator5
    Sampled Position Offset=&SelectVectorFromArray(Direct Array Index=$Particles.SpawnGroup, Vector Selection Array=$Emitter.BoneOffsetArray)
    Skeletal Mesh=$User.SkeletalMesh
    Socket Sampling Mode=::NewEnumerator5
   )
   /SetVariables(Particles.SpriteSize=<5 5>) {isSetParameters=true}
   /SetVariables(Particles.MaterialRandom:NiagaraFloat=&RandomRangeFloat(Minimum=0, Maximum=1)) {isSetParameters=true}
  ]
  particleUpdateScript [
   &ParticleState(Kill Particles When Lifetime Has Elapsed=false, Let Infinitely Lived Particles Die When Emitter Deactivates=true)
   &SkeletalMeshLocation(
    Bone / Socket Index=&SelectIntFromArray(Direct Array Index=$Particles.SpawnGroup, Int Selection Array=$Emitter.BoneSocketArray)
    MaskSpawnGroup=false
    Mesh Sampling Type=::NewEnumerator5
    Sampled Position Offset=&SelectVectorFromArray(Direct Array Index=$Particles.SpawnGroup, Vector Selection Array=$Emitter.BoneOffsetArray)
    Skeletal Mesh=$User.SkeletalMesh
    Socket Sampling Mode=::NewEnumerator5
   ) {module=SkeletalMeshLocation001}
   /SetVariables(Particles.BoneXAxis:Vector3f=&ReturnMeshOrientationAxis(Mesh Orientation Axis=::XAxis, Mesh Orientation Quat=$Particles.SkeletalMeshLocation001.SampledRotation)) {isSetParameters=true}
   &SolveForcesAndVelocity(Clamp Velocity=false, Limit Acceleration=false)
   /SetVariables(Particles.MaxRibbonLength:NiagaraFloat=&SelectFloatFromArray(Direct Array Index=$Particles.SpawnGroup, Float Selection Array=$Emitter.MaxRibbonLengthArray)) {isSetParameters=true}
   &GenerateLocationEvent(
    Event Generation Enabled:NiagaraBool=&Boolean_OrOperation(Bool1=&SetBoolByFloatComparison(A=$System.EngineOwnerSpeed, B=5), Bool2=$User.UseCinematicVelocity)
    Event Type:ENiagara_LocEventType=::NewEnumerator1
    "Float to Send as Float 1 (NormalizedAge)":NiagaraFloat=$Particles.MaxRibbonLength
    Per Unit Velocity Vector:Vector3f=&MakeCustomVectorFromBool(
     Bool=$User.UseCinematicVelocity
     True Vector=&Add_Vector(A=$User.CinematicVelocity, B=$Particles.SkeletalMeshLocation001.SampledVelocity)
     False Vector=$Particles.SkeletalMeshLocation001.SampledVelocity
    )
    Unit Spacing=5
    Use Event Probability=false
    UseEventDelay=false
    "Vector to Send as Vector 2 (Velocity)":Vector3f=$Particles.BoneXAxis
   ) {module=GenerateLocationEvent001}
  ]
  rendererClasses [&NiagaraSpriteRendererProperties]
  renderers [{rendererClass=&NiagaraSpriteRendererProperties}]
 }
 SpeedLine_Particles {
  emitterUpdateScript [
   &EmitterState()
   &SpawnRate(
    SpawnRate=&MakeCustomFloatFromBool(Bool=$User.Jumping, True Float=30, False Float=0)
    Use Spawn Probability=false
   ) {enabled=false}
   &SpawnPerUnit(Module.Use Max Movement Threshold=true, Module.Use Movement Tolerance=true, Use Spawn Probability=false)
  ]
  particleSpawnScript [
   /SetVariables(Particles.AlphaFadeIn:NiagaraFloat=&FloatFromCurve(
     FloatCurve=DataInterface() {propertyValues {Curve {keys [{tangentMode=RCTM_User, time=1, value=0}, {tangentMode=RCTM_User, time=2, value=0.1}]}}}
     CurveIndex=$Emitter.Age
    )) {isSetParameters=true}
   &InitializeParticle(
    Color=&RandomRangeLinearColor(Minimum=<rgba 1 0.9586 0.5853 0.1>, Maximum=<rgba 1 0.4475 0 0.15>)
    Lifetime Max=0.3
    Lifetime Min=0.2
    Lifetime Mode=::NewEnumerator1
    Sprite Size Max=<3 45>
    Sprite Size Min=<2 15>
    Sprite Size Mode=::NewEnumerator2
    UsePositionOffset=false
    Write Material Random=true
   )
   &SkeletalMeshLocation(MaskSpawnGroup=false, OverrideRandomSeed=false, Skeletal Mesh=$User.Skeletal Mesh)
   &ShapeLocation(
    MaskSpawnGroup=false
    Offset=<0 15 0>
    Offset Mode=::NewEnumerator0
    Override Seed=false
    Shape Origin:Vector3f=$StackContext.SkeletalMeshLocation.SampledPosition
    Sphere Radius=5
    UseOwnerScale=true
   )
   &AddVelocity(Velocity=&RandomRangeVector(Minimum=<-500 0 0>, Maximum=<-800 0 0>)) {enabled=false}
   &AddVelocity(Velocity=&NormalizeVector(Vector To Normalize=$Engine.Owner.Velocity, Safe Normalize=true, Safe Fallback Threshold=0, Safe Fallback Vector=<0 0 0.001>)) {enabled=false, module=AddVelocity001}
   &InheritVelocity(Inherited Velocity Speed Limit=300, Speed Limit=true, UseSpeedThreshold=false)
   &ApplyInitialForces()
  ]
  particleUpdateScript [
   &InheritSourceMovement(
    Applied Position Scale=<-1 -1 0>
    ApplyForce=true
    ApplyForceFalloff=false
    ApplyPosition=true
    ApplyPositionFalloff=true
    ApplyPositionFalloffCurve=true
    Module.LimitForceSpeed=false
   )
   &SolveForcesAndVelocity(Clamp Velocity=false, Limit Acceleration=false)
   &ParticleState()
   &ScaleColor(
    Scale Alpha=&Multiply_Float(
     A=&FloatFromCurve(
      FloatCurve=DataInterface() {propertyValues {Curve {keys [
          {arriveTangent=9.0725, interpMode=RCIM_Cubic, leaveTangent=9.0725, tangentMode=RCTM_User, time=0, value=0}
          {arriveTangent=-2.5249, interpMode=RCIM_Cubic, leaveTangent=-2.5249, tangentMode=RCTM_User, time=0.1408, value=0.9847}
          {arriveTangent=-0.0405, interpMode=RCIM_Cubic, leaveTangent=-0.0405, tangentMode=RCTM_User, time=1, value=0}
         ]}}}
      Scale Curve=15
     )
     B=$Particles.AlphaFadeIn
    )
    ScaleA=true
    ScaleRGB=true
   )
   &ScaleSpriteSize(
    Non-Uniform Curve Sprite Scale=DataInterface() {propertyValues {
      XCurve {keys [
        {interpMode=RCIM_Cubic, tangentMode=RCTM_User, time=0, value=0}
        {interpMode=RCIM_Cubic, tangentMode=RCTM_User, time=0.5, value=1}
        {tangentMode=RCTM_User, time=1, value=0}
       ]}
      YCurve {keys [
        {interpMode=RCIM_Cubic, tangentMode=RCTM_User, time=0, value=0}
        {interpMode=RCIM_Cubic, tangentMode=RCTM_User, time=0.5, value=1}
        {tangentMode=RCTM_User, time=1, value=0}
       ]}
     }}
    Scale Sprite Size Mode=::NewEnumerator3
   )
   &SpriteFacingAndAlignment(
    Sprite Alignment=&NormalizeVector(Vector To Normalize=$Engine.Owner.Velocity, Safe Normalize=true, Safe Fallback Threshold=0, Safe Fallback Vector=<0 0 0.001>)
    Sprite Facing=&NormalizeVector(Vector To Normalize=$Engine.Owner.Velocity, Safe Normalize=true, Safe Fallback Threshold=0, Safe Fallback Vector=<0 0 0.001>)
    WriteAlignment=true
    WriteFacing=true
   ) {enabled=false}
  ]
  rendererClasses [&NiagaraSpriteRendererProperties]
  renderers [{rendererClass=&NiagaraSpriteRendererProperties}]
 }
 SpeedyFeet {
  emitterUpdateScript [
   &EmitterState(Inactive Response=::NewEnumerator1, Life Cycle Mode=::NewEnumerator1, Loop Duration=1, UseLoopDelay=false)
   &SpawnBurst_Instantaneous(Spawn Count=4, Use Spawn Probability=false, UseLoopCountLimit=false)
  ]
  particleSpawnScript [
   &PartitionParticles(Number Of Partitions=4)
   &InitializeParticle(
    Color=&RandomRangeLinearColor(Minimum=<rgba 1 0.237 0 0.5>, Maximum=<rgba 1 0.5976 0 0.5>)
    Lifetime=5
    Sprite Size=<15 40>
    Sprite Size Mode=::NewEnumerator4
    UsePositionOffset=false
    Write Material Random=true
   )
  ]
  particleUpdateScript [
   &ParticleState()
   &SkeletalMeshLocation(
    Bone / Socket Index=$Particles.PartitionParticles.Partition
    Bone Sampling Mode=::NewEnumerator4
    MaskSpawnGroup=false
    Skeletal Mesh=$User.SkeletalMesh2
   )
   &CameraOffset(Camera Offset Amount=20)
   &ScaleColor(
    Scale Alpha=&Waveform("Waveform [1]"=::NewEnumerator0, "[1] Frequency"=1, "[1] Amplitude Min/Max"=<0.5 1>)
    ScaleA=true
    ScaleRGB=true
   )
  ]
  rendererClasses [&NiagaraSpriteRendererProperties]
  renderers [
   {rendererClass=&NiagaraSpriteRendererProperties}
   {rendererClass=&NiagaraSpriteRendererProperties, rendererIndex=1}
  ]
 }
 SpeedyFeet001 {
  emitterUpdateScript [
   &EmitterState(Inactive Response=::NewEnumerator1, Life Cycle Mode=::NewEnumerator1, Loop Duration=1, UseLoopDelay=false)
   &SpawnBurst_Instantaneous(Spawn Count=2, Use Spawn Probability=false, UseLoopCountLimit=false)
  ]
  particleSpawnScript [
   &InitializeParticle(
    Color=<rgba 1 0.337 0 0.5>
    Lifetime=5
    Sprite Size=<25 40>
    Sprite Size Mode=::NewEnumerator4
    UsePositionOffset=true
    Write Material Random=true
   )
   &PartitionParticles(Number Of Partitions=4)
   &SpriteFacingAndAlignment(WriteAlignment=false, WriteFacing=true)
  ]
  particleUpdateScript [
   &ParticleState()
   &SkeletalMeshLocation(
    Bone / Socket Index=$Particles.PartitionParticles.Partition
    Bone Sampling Mode=::NewEnumerator4
    MaskSpawnGroup=false
    Sampled Position Offset=<0 0 -3>
    Sampled Position Offset Coordinate Space:ENiagaraSkelMeshTransforms=::NewEnumerator0
    Skeletal Mesh=$User.SkeletalMesh2
   )
   &CameraOffset(Camera Offset Amount=5)
   &ScaleColor(
    Scale Alpha=&Waveform("Waveform [1]"=::NewEnumerator0, "[1] Frequency"=1, "[1] Amplitude Min/Max"=<0.5 1>)
    ScaleA=true
    ScaleRGB=true
   )
  ]
  rendererClasses [&NiagaraSpriteRendererProperties]
  renderers [{rendererClass=&NiagaraSpriteRendererProperties}]
 }
 Trail {
  emitterUpdateScript [
   &EmitterState()
   &SpawnRate(
    SpawnRate=&Multiply_Float(
     A=&FN_FrontEndFloatModifier(Gameplay=$User.ForwardDot, Front End=1)
     B=&UniformRangedFloat(Minimum=250, Maximum=500)
    )
    Use Spawn Probability=false
   ) {enabled=false}
   &SpawnPerUnit(Module.Use Max Movement Threshold=true, Module.Use Movement Tolerance=false, Spawn Spacing=1, Use Spawn Probability=false)
  ]
  particleSpawnScript [
   &FN_SetSpawnGroup(Number of Spawn Groups=2)
   &InitializeParticle(
    Color=<rgba 1 1 1 0.15>
    Lifetime=0.1
    Ribbon Width=20
    Ribbon Width Mode=::NewEnumerator1
    UsePositionOffset=false
    Write Material Random=true
   )
   &InitializeRibbon(
    Color=&ColorFromCurve(
     ColorCurve=DataInterface() {propertyValues {
       AlphaCurve {keys [<at 0.67 0.75 RCIM_Cubic>,<at 1 0 RCIM_Cubic>]}
       BlueCurve {keys [<at 0 1 RCIM_Cubic>,<at 1 1 RCIM_Cubic>]}
       GreenCurve {keys [<at 0 1 RCIM_Cubic>,<at 1 1 RCIM_Cubic>]}
       RedCurve {keys [<at 0 1 RCIM_Cubic>,<at 1 1 RCIM_Cubic>]}
      }}
     CurveIndex=$System.NormalizedLoopAge
    )
    Lifetime=0.5
    Module.Write Color=true
    Module.Write Lifetime=true
    Module.Write Mass=false
    Module.Write Material Random=true
    Module.Write Position=false
    Module.Write RibbonTwist=false
    Module.Write RibbonWidth=true
    Ribbon Width=40
   ) {enabled=false}
   &SkeletalMeshLocation(Bone / Socket Index=$Particles.SpawnGroup, Bone Sampling Mode=::NewEnumerator4, MaskSpawnGroup=false, Skeletal Mesh=$User.SkeletalMesh2)
   &PartitionParticles()
  ]
  particleUpdateScript [
   &ParticleState()
   &ScaleColor(
    Scale Alpha=&FloatFromCurve(FloatCurve=DataInterface() {propertyValues {Curve {keys [
         {interpMode=RCIM_Cubic, tangentMode=RCTM_User, time=0, value=0}
         {interpMode=RCIM_Cubic, tangentMode=RCTM_User, time=0.5, value=1}
         {tangentMode=RCTM_User, time=1, value=0}
        ]}}})
    Scale RGB=&MakeVectorFromLinearColorRGB(Color=&ColorFromCurve(ColorCurve=DataInterface() {propertyValues {
        AlphaCurve {keys [{arriveTangent=-0.9101, interpMode=RCIM_Cubic, time=0, value=1}]}
        BlueCurve {keys [<at 0.0019 0>, <at 0.2496 0>, {arriveTangent=-20.2868, time=0.583, value=1}]}
        GreenCurve {keys [<at 0.0019 0.303>, <at 0.2496 0.5647>, {arriveTangent=-13.8047, time=0.583, value=0.9897}]}
        RedCurve {keys [<at 0.0019 1>, <at 0.2496 1>, {arriveTangent=-1.2967, time=0.583, value=0.974}]}
       }}))
    ScaleA=true
    ScaleRGB=true
   )
   &ScaleRibbonWidth(Ribbon Width Scale=&FloatFromCurve(FloatCurve=DataInterface() {propertyValues {Curve {keys [
         {arriveTangent=9.0725, interpMode=RCIM_Cubic, leaveTangent=9.0725, tangentMode=RCTM_User, time=0, value=0}
         {arriveTangent=-2.5249, interpMode=RCIM_Cubic, leaveTangent=-2.5249, tangentMode=RCTM_User, time=0.1408, value=0.9847}
         {arriveTangent=-0.0405, interpMode=RCIM_Cubic, leaveTangent=-0.0405, tangentMode=RCTM_User, time=1, value=0}
        ]}}}))
   &DynamicMaterialParameters(Write Parameter Index 0=false)
  ]
  rendererClasses [&NiagaraRibbonRendererProperties]
  renderers [{rendererClass=&NiagaraRibbonRendererProperties}]
 }
}
summary {
 systemName=NS_Energy_LastingStamina_WildlifeMount_Raptor
 emitters [
  {emitterName=Trail, rendererClasses [&NiagaraRibbonRendererProperties]}
  {emitterName=SpeedLine_Particles, rendererClasses [&NiagaraSpriteRendererProperties]}
  {emitterName=SpeedyFeet, rendererClasses [&NiagaraSpriteRendererProperties]}
  {emitterName=SpeedyFeet001, rendererClasses [&NiagaraSpriteRendererProperties]}
  {emitterName=RibbonSourcesLegs, rendererClasses [&NiagaraSpriteRendererProperties]}
  {
   emitterName=LegRibbons002
   rendererClasses [&NiagaraRibbonRendererProperties, &NiagaraSpriteRendererProperties]
  }
 ]
 userVariables [
  User.BariumPaintColor {type=LinearColor, description="", defaultValue {struct=LinearColor, value=<rgba 1 1 1 1>}}
  User.bIsFrontEnd {type=NiagaraBool}
  User.bIsFrontEndPreview {type=NiagaraBool}
  User.bIsLocalClient {type=NiagaraBool}
  User.CinematicVelocity {type=Vector3f, description="", defaultValue {struct=Vector3f, value=<0 200 200>}}
  User.energy {type=NiagaraFloat, defaultValue=0, description=""}
  User.ForwardDot {type=NiagaraFloat, defaultValue=0, description=""}
  User.Jumping {type=NiagaraBool, defaultValue=true}
  User.Primary Color {type=LinearColor, description="", defaultValue {struct=LinearColor, value=<rgba 1 1 1 1>}}
  User.RibbonIDOffset {type=NiagaraInt32}
  User.RibbonLengthFactor {type=NiagaraFloat, defaultValue=0, description=""}
  User.Skeletal Mesh {
   type=NiagaraDataInterfaceSkeletalMesh
   defaultValue {
    struct=NiagaraExt_VariableValue_DataInterface
    value {dataInterface=&self:NiagaraDataInterfaceSkeletalMesh_0}
   }
  }
  User.SkeletalMesh {
   type=NiagaraDataInterfaceSkeletalMesh
   defaultValue {
    struct=NiagaraExt_VariableValue_DataInterface
    value {dataInterface=&self:NiagaraDataInterfaceSkeletalMesh_3}
   }
  }
  User.SkeletalMesh2 {
   type=NiagaraDataInterfaceSkeletalMesh
   defaultValue {
    struct=NiagaraExt_VariableValue_DataInterface
    value {dataInterface=&self:NiagaraDataInterfaceSkeletalMesh_1}
   }
  }
  User.UseCinematicVelocity {type=NiagaraBool}
 ]
}
