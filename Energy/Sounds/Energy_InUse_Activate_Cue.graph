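# Sound cue for activating an Energy bottle item, classed as gadget and item audio.
# A mixer layers two EnergyStart waves, each at a fixed raised pitch, with a slightly delayed EnergySlap wave at differing input levels.
# A root modulator randomizes the mix's pitch upward on each play.
# The cue uses status effect activation attenuation, while the slap branch carries its own consumable-use attenuation node.
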SoundCue /Energy/Sounds/Energy_InUse_Activate_Cue
root=Modulator_0
nodes [
 {id=Attenuation_0, children [Delay_0], properties {attenuationSettings=&Fort_Consumable_Use_Athena_Att}}
 {id=Delay_0, children [WavePlayer_1], properties {delayMax=0.2, delayMin=0.2}}
 {id=Mixer_0, children [Modulator_2, Modulator_3, Attenuation_0], properties {inputVolume [0.6,1,1.1]}}
 {id=Modulator_0, children [Mixer_0], properties {pitchMax=1.1, pitchMin=1}}
 {id=Modulator_2, children [WavePlayer_0], properties {pitchMax=1.2, pitchMin=1.2, volumeMax=1, volumeMin=1}}
 {id=Modulator_3, children [WavePlayer_2], properties {pitchMax=1.1, pitchMin=1.1, volumeMax=1, volumeMin=1}}
 {id=WavePlayer_0, properties {soundWaveAssetPtr=&/Energy_Bottle_EnergyStart_01}}
 {id=WavePlayer_1, properties {soundWaveAssetPtr=&/Energy_Bottle_EnergySlap_01}}
 {id=WavePlayer_2, properties {soundWaveAssetPtr=&/Energy_Bottle_EnergyStart_02}}
]
properties {
 attenuationSettings=/Game/Sounds/Attenuation/Fort_Ability_Status_Effect_Activation_Att
 maxDistance=6000
 soundClassObject=&Fort_Gadgets_Items
 volumeMultiplier=0.5
}
