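# Looping sound cue for an energy item in use, classed as a gadget/item sound.
# A Branch on the IsLocal parameter picks stereo status-effect attenuation when true, and mono attenuation when false or unset.
# Each path runs through its own Modulator to a looping player of an Energy_Bottle_EnergyLoop_01 wave.
# The cue itself lowers the volume and slightly raises the pitch.
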SoundCue /Energy/Sounds/Energy_InUse_Loop_Cue
root=Branch_0
nodes [
 {
  id=Attenuation_0
  children [Modulator_1]
  properties {attenuationSettings=/FNGameplayCues/Sounds/Attenuation/Fort_Ability_Status_Effect_Looping_Stereo_Att}
 }
 {
  id=Attenuation_1
  children [Modulator_0]
  properties {attenuationSettings=/FNGameplayCues/Sounds/Attenuation/Fort_Ability_Status_Effect_Looping_Mono_Att}
 }
 {id=Branch_0, children [Attenuation_0, Attenuation_1, Attenuation_1], properties {boolParameterName=IsLocal}}
 {id=Modulator_0, children [WavePlayer_0]}
 {id=Modulator_1, children [WavePlayer_2]}
 {id=WavePlayer_0, properties {bLooping=true, soundWaveAssetPtr=&/Energy_Bottle_EnergyLoop_01}}
 {id=WavePlayer_2, properties {bLooping=true, soundWaveAssetPtr=&Energy_Bottle_EnergyLoop_01}}
]
properties {maxDistance=3500, pitchMultiplier=1.1, soundClassObject=&Fort_Gadgets_Items, volumeMultiplier=0.2}
