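# Sound cue for the deactivation of an in-use Energy weapon, playing an energy-stop one-shot from a bottle sound asset.
# A Modulator feeds the single wave player, locking pitch at a fixed raised value rather than randomizing it.
# The cue plays quietly under the gadgets and items sound class, using a status-effect activation attenuation.
# A concurrency override limits it to one instance at a time.
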SoundCue /Energy/Sounds/Energy_Weapon_InUse_Deactivated_Cue
root=Modulator_1
nodes [
 {id=Modulator_1, children [WavePlayer_0], properties {pitchMax=1.2, pitchMin=1.2}}
 {id=WavePlayer_0, properties {soundWaveAssetPtr=&/Energy_Bottle_EnergyStop_01}}
]
properties {
 attenuationSettings=/Game/Sounds/Attenuation/Fort_Ability_Status_Effect_Activation_Att
 bOverrideConcurrency=true
 maxDistance=6000
 soundClassObject=&Fort_Gadgets_Items
 volumeMultiplier=0.1
 concurrencyOverrides {maxCount=1}
}
