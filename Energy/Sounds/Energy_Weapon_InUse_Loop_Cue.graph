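# Sound cue for an energy weapon's in-use sound, layering start-up one-shots over a looping energy bed.
# An IsLocal branch selects stereo attenuation when true, or mono attenuation when false or unset.
# The local path mixes two start sounds with an enveloped loop that fades in, then gradually drops.
# The other path uses one start sound and a loop that fades in and holds, with modulators pitching both paths up.
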SoundCue /Energy/Sounds/Energy_Weapon_InUse_Loop_Cue
root=Branch_2
nodes [
 {
  id=Attenuation_2
  children [Mixer_1]
  properties {attenuationSettings=/FNGameplayCues/Sounds/Attenuation/Fort_Ability_Status_Effect_Looping_Stereo_Att}
 }
 {
  id=Attenuation_3
  children [Mixer_3]
  properties {attenuationSettings=/FNGameplayCues/Sounds/Attenuation/Fort_Ability_Status_Effect_Looping_Mono_Att}
 }
 {id=Branch_2, children [Attenuation_2, Attenuation_3, Attenuation_3], properties {boolParameterName=IsLocal}}
 {
  id=Enveloper_0
  children [Modulator_1]
  properties {
   pitchCurve {editorCurveData {keys [{time=0}]}}
   volumeCurve {editorCurveData {keys [<at 0 0>, <at 1 1>, {arriveTangent=-0.0778, interpMode=RCIM_Cubic, leaveTangent=-0.0778, time=1.15, value=1}, <at 10 0.3>]}}
  }
 }
 {
  id=Enveloper_1
  children [Modulator_4]
  properties {pitchCurve {editorCurveData {keys [{time=0}]}}, volumeCurve {editorCurveData {keys [<at 0 0>,<at 1 1>,<at 1.6799 1>,<at 2.2189 1>]}}}
 }
 {id=Mixer_0, children [Modulator_2,Modulator_3,null], properties {inputVolume [0.6,1,1.1]}}
 {id=Mixer_1, children [Modulator_0,Enveloper_0], properties {inputVolume [0.5,0.3]}}
 {id=Mixer_2, children [Modulator_6,null,null], properties {inputVolume [0.6,1,1.1]}}
 {id=Mixer_3, children [Modulator_5,Enveloper_1], properties {inputVolume [0.5,0.3]}}
 {id=Modulator_0, children [Mixer_0], properties {pitchMax=1.1, pitchMin=1}}
 {id=Modulator_1, children [WavePlayer_2], properties {pitchMax=1.3, pitchMin=1.2}}
 {id=Modulator_2, children [WavePlayer_0], properties {pitchMax=1.2, pitchMin=1.2, volumeMax=1, volumeMin=1}}
 {id=Modulator_3, children [WavePlayer_1], properties {pitchMax=1.1, pitchMin=1.1, volumeMax=1, volumeMin=1}}
 {id=Modulator_4, children [WavePlayer_4], properties {pitchMax=1.3, pitchMin=1.2}}
 {id=Modulator_5, children [Mixer_2], properties {pitchMax=1.1, pitchMin=1}}
 {id=Modulator_6, children [WavePlayer_5], properties {pitchMax=1.2, pitchMin=1.2, volumeMax=1, volumeMin=1}}
 {id=WavePlayer_0, properties {soundWaveAssetPtr=&/Energy_Bottle_EnergyStart_01}}
 {id=WavePlayer_1, properties {soundWaveAssetPtr=&/Energy_Bottle_EnergyStart_02}}
 {id=WavePlayer_2, properties {bLooping=true, soundWaveAssetPtr=&/Energy_Bottle_EnergyLoop_01}}
 {id=WavePlayer_4, properties {bLooping=true, soundWaveAssetPtr=&Energy_Bottle_EnergyLoop_01}}
 {id=WavePlayer_5, properties {soundWaveAssetPtr=&Energy_Bottle_EnergyStart_01}}
]
properties {maxDistance=3500, soundClassObject=&Fort_Gadgets_Items, volumeMultiplier=0.6}
